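; Guided-motion movement mode extension for the Rift Gauntlets dash, branching on a dash type read from networked optional floats.
; Activation sends dash-type events to the dash ability and force-ends the mode on PreLaunch knockbacks, optionally capping upward velocity.
; Overrides give side dashes custom hover height and slams their own force, duration and angle-limited downward direction.
; The pawn component clamps root-motion Z; non-side dashes cap finish Z near ground, detected by montage section or trace.
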(event EventOnExtensionActivate
  (params (RuntimeContext :type MMERuntime%20Context%20Structure%20%28by%20ref%29))
  (bind (_velocity_x _velocity_y _velocity_z) (Math|Vector|BreakVector (Variables|/Script/Engine.MovementComponent|GetVelocity_X (Variables|FortMovementMode_BaseExtLogic|GetOwnerMovementComponent))))
  ;[0.01374 0.15 0.1201 0.5] OnExtensionActivate/Deactivate: LaunchCharacter knockbacks should cancel the MME.
  (|CallParentFunction RuntimeContext)
  (CallFunction|ExecuteGameplayCueNotReplicated (Utilities|Struct|BreakGameplayCueTag (Variables|GameplayCue|GetActivateGameplayCueTag)))
  (bind _returnvalue (Class|FortMovementModeExtBlueprintLibrary|GetRuntimeDatafromContext RuntimeContext "/Script/FortniteGame.FortMovementMode_BaseExtRuntimeData"))
  (bind _output_get (Variables|Transient|SetRuntimeData _returnvalue))
  (bind _asfort_movement_mode_guided_motion_runtime_data (Utilities|Casting|CastToFortMovementMode_GuidedMotionRuntimeData _output_get))
  (Utilities|FlowControl|Sequence
    (:then_0
      (bind _ownerpawn (Variables|FortMovementMode_BaseExtLogic|GetOwnerPawn)
        (:then
          (bind _asynctaskproxy (Ability|Async|WaitGameplayEventToActor _ownerpawn "(TagName=\"Event.Character.PreLaunch\")")
            (:then
              (Variables|Transient|SetWaitGameplayEventAction _asynctaskproxy))
            (:EventReceived
              (bind _runtimedata (Variables|Transient|GetRuntimeData)
                (:then
                  (Class|FortMovementModeBaseExtRuntimeData|ForceManualEnd _runtimedata)
                  ; Clamp velocity Z when going to be launched, in case we have weirdly large Z speed from calculation
                  (bind _returnvalue_2 (Development|GetConsoleVariableFloatValue "RiftGauntlets.DashLaunch.ClampZVelocity"))
                  (if (> _returnvalue_2 0)
                    (Variables|/Script/Engine.MovementComponent|SetVelocity_X (Math|Vector|MakeVector _velocity_x _velocity_y (Math|Float|Min(Float) _velocity_z _returnvalue_2)) (Variables|FortMovementMode_BaseExtLogic|GetOwnerMovementComponent))))
                (:else)))))
        (:else)))
    (:then_1
      ;;[0.02667 0.15 0.1121 0.5] Sending gameplay event with payload letting listeners know if side dash is initiated or not.
      (bind _ownerpawn_2 (Variables|FortMovementMode_BaseExtLogic|GetOwnerPawn)
        (:then
          (Ability|SendGameplayEventtoActor _ownerpawn_2 "(TagName=\"RiftGauntlets.Event.Ability.Dash.MME.OptionalFloat.0\")" (Utilities|Struct|MakeGameplayEventData "(TagName=\"\")" 0 0 0 0 "()" "(GameplayTags=)" "(GameplayTags=)" (Math|Conversions|ToFloat(Integer) (Math|Conversions|ToInteger(Byte) (HelperFunctions|GetDashType _asfort_movement_mode_guided_motion_runtime_data))))) ; This lets the dash GA know about the dash type.
          (Ability|SendGameplayEventtoActor _ownerpawn_2 "(TagName=\"RiftGauntlets.Event.Ability.Dash.MME.StartPostDashTypeSet\")") ; Dash GA waits for this call before it registers that the dash MME has started. \nThis is so that we can reliably use the current dash type as know it has been set.
        )
        (:else)))))

(event EventOnExtensionDeactivate
  (params (RuntimeContext :type MMERuntime%20Context%20Structure%20%28by%20ref%29))
  ;[0.01374 0.15 0.1201 0.5] OnExtensionActivate/Deactivate: LaunchCharacter knockbacks should cancel the MME.
  (|CallParentFunction RuntimeContext)
  (bind _waitgameplayeventaction (Variables|Transient|GetWaitGameplayEventAction)
    (:then
      (Ability|Async|EndAction _waitgameplayeventaction))
    (:else)))

(fn BP_AdjustRootMotionVelocity
  (params (RuntimeContext :type MMERuntime%20Context%20Structure%20%28by%20ref%29) (InVelocity :type Vector :default "0, 0, 0"))
  ;[0.1328 0.0394 0.15 0.5] AdjustRootMotionVelocity: Overridden to clamp the velocity on the Z-axis based on height above ground / height above build grid.
  (Utilities|FlowControl|Sequence
    (:then_0
      (bind _ownerpawn (Variables|FortMovementMode_BaseExtLogic|GetOwnerPawn)
        (:then
          (bind _returnvalue (Actor|GetComponentbyClass _ownerpawn "/RiftGauntletsGameplay/Gameplay/B_RiftGauntletsPawnComponent.B_RiftGauntletsPawnComponent_C"))
          (Utilities|IsValid _returnvalue
            (:Is%20Valid
              (bind _output_get (LocalVariables|SetVelocityZ (.z InVelocity)))
              (bind _returnvalue_2 (Class|FortPawnComponentRiftGauntlets|PerformMovementModeClampMaxVelocityZ _returnvalue _output_get))
              (if _returnvalue_2
                (return (Math|Vector|MakeVector (.x InVelocity) (.y InVelocity) _output_get))))))
        (:else)))
    (:then_1
      (return InVelocity))))

(fn CalculateFinishVelocity
  (params (GuidedMotionRuntimeData :type Fort%20Movement%20Mode%20Guided%20Motion%20Runtime%20Data%20Object%20Reference) (InVelocity :type Vector%20%28by%20ref%29 :default "0, 0, 0"))
  ;[0.01556 0.04821 0.15 0.5] CalculateFInishVelocity: Adjust velocity if forward/punching dash is ending close to the ground.
  (bind _returnvalue (|CallParentFunction GuidedMotionRuntimeData InVelocity))
  (LocalVariables|SetReturnVelocity _returnvalue)
  (Utilities|FlowControl|Sequence
    (:then_0
      (if (not (Utilities|Enum|Equal(Enum) (HelperFunctions|GetDashType GuidedMotionRuntimeData) "NewEnumerator1"))
        (bind _returnvalue_2 (HelperFunctions|TraceClosetoGround))
        (if _returnvalue_2
          (LocalVariables|SetReturnVelocity (Math|Vector|MakeVector (.x _returnvalue) (.y _returnvalue) (Math|Float|Min(Float) (.z _returnvalue) (ScalableFloat|GetValueatLevel (Variables|Hotfix|GetHotfix_DashOnGroundVelocityMaxClamp))))))))
    (:then_1
      (return (LocalVariables|GetReturnVelocity)))))

(fn GetDashType
  (params (RuntimeData :type Fort%20Movement%20Mode%20Guided%20Motion%20Runtime%20Data%20Object%20Reference))
  (bind _netoptionalfloats (Variables|/Script/FortniteGame.FortMovementMode_GuidedMotionRuntimeData|GetNetOptionalFloats RuntimeData))
  (Utilities|FlowControl|Sequence
    (:then_0
      (Utilities|IsValid RuntimeData
        (:Is%20Valid
          (if (Utilities|Array|IsValidIndex _netoptionalfloats)
            (return (Utilities|Enum|BytetoEnumRiftGauntletsDashTypeEnum (Math|Conversions|ToByte(Integer) (Math|Float|Truncate (Utilities|Struct|BreakFortMovementModeFloatPropertyNetworked (Utilities|Array|Get(aref) _netoptionalfloats))))))))))
    (:then_1
      (return "NewEnumerator0"))))

(fn GetDirection_ConstantForce
  (params (GuidedMotionRuntimeData :type Fort%20Movement%20Mode%20Guided%20Motion%20Runtime%20Data%20Object%20Reference))
  (bind _hotfix_slamintentangle (Variables|Hotfix|GetHotfix_SlamIntentAngle))
  (bind _returnvalue (ScalableFloat|GetValueatLevel _hotfix_slamintentangle))
  (bind _returnvalue_2 (|CallParentFunction GuidedMotionRuntimeData))
  (Utilities|FlowControl|Sequence
    (:then_0
      (if (Utilities|Enum|Equal(Enum) (HelperFunctions|GetDashType GuidedMotionRuntimeData) "NewEnumerator2")
        (if (> (Math|Trig|Acos(Degrees) (Math|Vector|DotProduct (Math|Vector|Normalize _returnvalue_2) (Math|Vector|VectorDown))) _returnvalue)
          (return (* (Math|Vector|RotateVector (Math|Vector|VectorDown) (Math|Vector|RotatorfromAxisandAngle (Math|Vector|Normalize (Math|Vector|CrossProduct (Math|Vector|VectorDown) _returnvalue_2)) _returnvalue)) (Math|Vector|VectorLength _returnvalue_2))))))
    (:then_1
      (return _returnvalue_2))))

(fn GetEnforceHeightAboveGround_ConstantForce
  (params (GuidedMotionRuntimeData :type Fort%20Movement%20Mode%20Guided%20Motion%20Runtime%20Data%20Object%20Reference))
  ;[0.15 0.02306 0.1386 0.5] Enforce height above ground only when we're not slam dashing.
  (return (Utilities|Enum|NotEqual(Enum) (HelperFunctions|GetDashType GuidedMotionRuntimeData) "NewEnumerator2")))

(fn GetHeightAboveGround_ConstantForce
  (params (GuidedMotionRuntimeData :type Fort%20Movement%20Mode%20Guided%20Motion%20Runtime%20Data%20Object%20Reference))
  ;[0.1455 0.0542 0.15 0.5] Side dash implements a custom height above ground setting.
  (Utilities|FlowControl|Sequence
    (:then_0
      (if (Utilities|Enum|Equal(Enum) (HelperFunctions|GetDashType GuidedMotionRuntimeData) "NewEnumerator1")
        (return (ScalableFloat|GetValueatLevel (Variables|Hotfix|GetHotfix_SideDashHeightAboveGround)))))
    (:then_1
      (bind _returnvalue (|CallParentFunction GuidedMotionRuntimeData))
      (return _returnvalue))))

(fn GetMaxDuration
  (params (GuidedMotionRuntimeData :type Fort%20Movement%20Mode%20Guided%20Motion%20Runtime%20Data%20Object%20Reference))
  (bind _returnvalue (|CallParentFunction GuidedMotionRuntimeData))
  (return (Math|Float|SelectFloat (ScalableFloat|GetValueatLevel (Variables|Hotfix|GetHotfix_SlamMaxDuration)) _returnvalue (Utilities|Enum|Equal(Enum) (HelperFunctions|GetDashType GuidedMotionRuntimeData) "NewEnumerator2"))))

(fn GetMovementDirectionXY_WhenNotUsingAccelerationForTarget
  (bind _ownermovementcomponent (Variables|FortMovementMode_BaseExtLogic|GetOwnerMovementComponent)
    (:then
      (return (Math|Vector|GetForwardVector (CharacterMovement|GetNetSyncedBaseAimRotation _ownermovementcomponent))))
    (:else
      (return (|CallParentFunction)))))

(fn GetStrength_ConstantForce
  (params (GuidedMotionRuntimeData :type Fort%20Movement%20Mode%20Guided%20Motion%20Runtime%20Data%20Object%20Reference))
  (bind _returnvalue (|CallParentFunction GuidedMotionRuntimeData))
  (return (Math|Float|SelectFloat (ScalableFloat|GetValueatLevel (Variables|Hotfix|GetHotfix_SlamConstantForce)) _returnvalue (Utilities|Enum|Equal(Enum) (HelperFunctions|GetDashType GuidedMotionRuntimeData) "NewEnumerator2"))))

(fn TraceCloseToGround
  ;[0.05596 0.1641 0.2083] Returns if the player is close to the ground.
  (Utilities|FlowControl|Sequence
    (:then_0
      (bind _ownerpawn (Variables|FortMovementMode_BaseExtLogic|GetOwnerPawn)
        (:then
          (bind _mesh (Variables|/Script/Engine.Character|GetMesh _ownerpawn)
            (:then
              (Utilities|IsValid (Components|SkeletalMesh|GetAnimInstance _mesh)
                (:Is%20Valid
                  (if (or (== (Animation|Montage|MontageGetCurrentSection (Components|SkeletalMesh|GetAnimInstance _mesh) 0) "GroundSlam") (== (Animation|Montage|MontageGetCurrentSection (Components|SkeletalMesh|GetAnimInstance _mesh) 0) "GroundPunch"))
                    (return true)))))
            (:else)))
        (:else)))
    (:then_1
      (bind _ownerpawn_2 (Variables|FortMovementMode_BaseExtLogic|GetOwnerPawn)
        (:then
          (bind _output_get (LocalVariables|SetTraceStartLocation (Transformation|GetActorLocation _ownerpawn_2)))
          (LocalVariables|SetTraceEndLocation (Utilities|Operators|Subtract _output_get (Math|Vector|MakeVector 0.0 0.0 (+ (Components|Capsule|GetScaledCapsuleHalfHeight (Variables|/Script/Engine.Character|GetCapsuleComponent _ownerpawn_2)) (ScalableFloat|GetValueatLevel (Variables|Hotfix|GetHotfix_GroundDistanceCheck))))))
          (bind (_outhit _returnvalue) (Collision|LineTraceByChannel _ownerpawn_2 (LocalVariables|GetTraceStartLocation) (LocalVariables|GetTraceEndLocation) "TraceTypeQuery10" true (Utilities|Array|MakeArray _ownerpawn_2) "None" true "(R=1.000000,G=0.000000,B=0.000000,A=1.000000)" "(R=0.000000,G=1.000000,B=0.000000,A=1.000000)" 1e+01))
          (return _returnvalue))
        (:else)))
    (:then_2
      (return false))))
