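; Guided-motion movement mode extension for the hover phase of the Rift Gauntlets dash.
; On activation it caches its runtime data and listens on the owner pawn for Event.Character.PreLaunch, forcing a manual end so LaunchCharacter knockbacks cancel the mode.
; On deactivation it ends that async listener.
; Constant-force direction combines the normalized horizontal velocity of the owner's movement component with a fixed upward Z of 1, without renormalizing.
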(event EventOnExtensionActivate
  (params (RuntimeContext :type MMERuntime%20Context%20Structure%20%28by%20ref%29))
  ;[0.01374 0.15 0.1201 0.5] OnExtensionActivate/Deactivate: LaunchCharacter knockbacks should cancel the MME.
  (|CallParentFunction RuntimeContext)
  (bind _returnvalue (Class|FortMovementModeExtBlueprintLibrary|GetRuntimeDatafromContext RuntimeContext "/Script/FortniteGame.FortMovementMode_BaseExtRuntimeData"))
  (Variables|Transient|SetRuntimeData _returnvalue)
  (bind _ownerpawn (Variables|FortMovementMode_BaseExtLogic|GetOwnerPawn)
    (:then
      (bind _asynctaskproxy (Ability|Async|WaitGameplayEventToActor _ownerpawn "(TagName=\"Event.Character.PreLaunch\")")
        (:then
          (Variables|Transient|SetWaitGameplayEventAction _asynctaskproxy))
        (:EventReceived
          (bind _runtimedata (Variables|Transient|GetRuntimeData)
            (:then
              (Class|FortMovementModeBaseExtRuntimeData|ForceManualEnd _runtimedata))
            (:else)))))
    (:else)))

(event EventOnExtensionDeactivate
  (params (RuntimeContext :type MMERuntime%20Context%20Structure%20%28by%20ref%29))
  ;[0.01374 0.15 0.1201 0.5] OnExtensionActivate/Deactivate: LaunchCharacter knockbacks should cancel the MME.
  (|CallParentFunction RuntimeContext)
  (bind _waitgameplayeventaction (Variables|Transient|GetWaitGameplayEventAction)
    (:then
      (Ability|Async|EndAction _waitgameplayeventaction))
    (:else)))

(fn GetDirection_ConstantForce
  (params (GuidedMotionRuntimeData :type Fort%20Movement%20Mode%20Guided%20Motion%20Runtime%20Data%20Object%20Reference))
  (return (.x (Math|Vector|Normalize2D(Vector) (Variables|/Script/Engine.MovementComponent|GetVelocity (Variables|FortMovementMode_BaseExtLogic|GetOwnerMovementComponent)))) (.y (Math|Vector|Normalize2D(Vector) (Variables|/Script/Engine.MovementComponent|GetVelocity (Variables|FortMovementMode_BaseExtLogic|GetOwnerMovementComponent)))) 1.0))
