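; Guided motion movement mode that carries a Rift Gauntlets dash toward a target actor.
; Max duration is derived from speed instead of fixed time: initial distance to the actor divided by the configured dash distance over configured duration.
; The target offset keeps the parent's X and Z, choosing a hotfixable right or left Y from the pawn's dash stance tag.
; Without an owner pawn, the parent offset passes through unchanged.
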(fn GetMaxDuration
  (params (GuidedMotionRuntimeData :type Fort%20Movement%20Mode%20Guided%20Motion%20Runtime%20Data%20Object%20Reference))
  ;[0.04556 0.15 0.1371 0.5] The duration to get to MoveToTarget actor should be based on the speed and not on time, so dynamically calculate the duration input based on the speed the dash is expected to use (Move distance / Duration)
  (return (Math|Float|SafeDivide (Math|Vector|Distance(Vector) (Variables|/Script/FortniteGame.FortMovementMode_GuidedMotionRuntimeData|GetLocalInitialPlayerLocation GuidedMotionRuntimeData) (Variables|/Script/FortniteGame.FortMovementMode_GuidedMotionRuntimeData|GetNetInitialWorldLocation_MoveToActor GuidedMotionRuntimeData)) (* (ScalableFloat|GetCachedScalableFloatValue (Variables|DashDistanceConfig|GetMovementDeltaDistance)) (Math|Float|SafeDivide 1.0 (ScalableFloat|GetValueatLevel (Variables|Config-Shared|GetMaxDuration)))))))

(fn GetTargetOffsetVector
  (params (GuidedMotionRuntimeData :type Fort%20Movement%20Mode%20Guided%20Motion%20Runtime%20Data%20Object%20Reference))
  ; Base the target offset vector on the dash stance.
  (bind _returnvalue (|CallParentFunction GuidedMotionRuntimeData))
  (Utilities|FlowControl|Sequence
    (:then_0
      (bind _ownerpawn (Variables|FortMovementMode_BaseExtLogic|GetOwnerPawn)
        (:then
          (return (.x _returnvalue) (ScalableFloat|GetValueatLevel (select (GameplayTags|HasMatchingGameplayTag _ownerpawn "(TagName=\"RiftGauntlets.Stance.Dash.Right\")") (Variables|Hofix|GetHotfix_TargetOffsetYRight) (Variables|Hofix|GetHotfix_TargetOffsetYLeft))) (.z _returnvalue)))
        (:else)))
    (:then_1
      (return _returnvalue))))
